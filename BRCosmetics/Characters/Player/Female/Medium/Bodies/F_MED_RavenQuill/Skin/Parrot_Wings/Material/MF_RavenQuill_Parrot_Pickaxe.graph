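# Material function for the Raven Quill Parrot pickaxe skin that layers animated, color-ramped flames over incoming material attributes.
# Pre-skinned position becomes offset, rotated radial UVs, distorted by a fire flow map to drive panning noise.
# Noise, flow and gradients index two lookup textures, blended and HDR-boosted, with a separate mobile path.
# A backpack FX mask blends the result into emissive, base color, metallic, roughness and a flow-distorted normal.
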MaterialFunction /BRCosmetics/Characters/Player/Female/Medium/Bodies/F_MED_RavenQuill/Skin/Parrot_Wings/Material/MF_RavenQuill_Parrot_Pickaxe
graph {nodes [
  n1 = PreSkinnedPosition()
  n2 = Add(
   A=ComponentMask(None=VertexInterpolator(VS=@n1).PS) {b=true, g=true}
   B=AppendVector(
    A=ScalarParameter() {defaultValue=16.6616, parameterName=UV_Offset_Gradient_X}
    B=ScalarParameter() {defaultValue=-31.4085, parameterName=UV_Offset_Gradient_Y}
   )
  )
  n3 = ScalarParameter() {defaultValue=2.2861, parameterName=UV_Distort_Rotator1}
  n4 = ScalarParameter() {defaultValue=30.2315, parameterName=UV_Size_WingGradient_X}
  n5 = ScalarParameter() {defaultValue=87.1347, parameterName=UV_Size_WingGradient_Y}
  n6 = MaterialFunctionCall(Vector or UVs=Divide(A=Rotator(Coordinate=@n2, Time=@n3) {centerX=0, centerY=0}, B=AppendVector(A=@n4, B=@n5))) {materialFunction=&VectorToRadialValue}
  n7 = Divide(
   A=@n6.Radial Coordinates
   B=AppendVector(
    A=ScalarParameter() {defaultValue=0.062, parameterName=UV_Size_WingDistortionX}
    B=ScalarParameter() {defaultValue=-0.6622, parameterName=UV_Size_WingDistortionY}
   )
  )
  ComponentMask(None=@n7) {g=true}
  n8 = Multiply(A=Time(), B=ScalarParameter() {defaultValue=0.75, group=FlameShape, parameterName=Speed_distortion})
  Frac(None=ComponentMask(None=Panner(Coordinate=@n6.Radial Coordinates, Time=@n8) {bFractionalPart=true, speedY=-1}) {g=true})
  n9 = GetMaterialAttributes(None=FunctionInput() {inputType=FunctionInput_MaterialAttributes}) {attributeGetTypes [
    #1
    #2
    #3
    #4
    #5
   ]}
  n10 = TextureSample() {
   samplerSource=SSM_Wrap_WorldGroupSettings
   samplerType=SAMPLERTYPE_Masks
   texture=/BRCosmetics/Characters/Player/Female/Medium/Bodies/F_MED_RavenQuill/Skin/Parrot_Wings/Texture/T_F_MED_Raven_Quill_Parrot_Backpack_FX
  }
  n11 = Saturate(None=Add(A=@n10.R, B=@n10.R))
  n12 = ScalarParameter() {defaultValue=0.937, parameterName=Gradient_Tweak_1}
  n13 = ScalarParameter() {defaultValue=0.3549, parameterName=Gradient_Tweak_2}
  n14 = ComponentMask(None=@n6.Radial Coordinates) {g=true}
  n15 = ComponentMask(None=VectorParameter() {group=FlameShape, parameterName=Flames_UVScale, defaultValue=<rgba 1.1 0.5 0 0>}.RGB) {g=true, r=true}
  n16 = Multiply(A=Time(), B=ScalarParameter() {defaultValue=0.25, parameterName=DistortUVAnim})
  n17 = Multiply(A=@n7, B=ComponentMask(None=VectorParameter() {parameterName=DistortUVScale, defaultValue=<rgba 0.8 0.5 0 0>}.RGB) {g=true, r=true})
  n18 = ComponentMask(None=TextureSample(UVs=Add(A=@n16, B=@n17)) {samplerType=SAMPLERTYPE_LinearColor, texture=/Game/Effects/Fort_Effects/Creatures/Elemental/Fire/MergedFireFlowMap}.RGB) {g=true, r=true}
  n19 = Multiply(A=@n18, B=ScalarParameter() {defaultValue=0.9578, parameterName=Distort}) {constB=0.5}
  n20 = Multiply(A=@n15, B=Add(A=@n7, B=@n19))
  n21 = ScalarParameter() {group=FlameShape, parameterName=Flame_UVOffset}
  n22 = Multiply(A=Time(), B=ScalarParameter() {defaultValue=0.15, group=FlameShape, parameterName=Flame_Speed})
  n23 = TextureSample(UVs=Panner(Coordinate=Add(A=@n20, B=@n21), Time=@n22) {speedX=0.2, speedY=1}) {texture=&T_NoisePack_01}
  n24 = Add(A=Multiply(A=@n7, B=@n15), B=@n21)
  n25 = MaterialFunctionCall(
   In=TextureSample(UVs=Panner(Coordinate=@n24, Time=@n8) {speedX=-0.5, speedY=1}) {texture=&T_NoisePack_01}.G
   Contrast=Constant() {r=0.2}
  ) {materialFunction=&CheapContrast}
  n26 = Add(A=@n7, B=@n19)
  n27 = Panner(Coordinate=Add(A=@n21, B=Multiply(A=@n15, B=@n26)), Time=@n22) {speedX=-0.2, speedY=1}
  n28 = MaterialFunctionCall() {materialFunction=&LinearGradient}
  n29 = ScalarParameter() {defaultValue=0.1847, group=FlameShape, parameterName=FireLine}
  n30 = Add(
   A=Multiply(A=Add(A=@n25.Result, B=@n23.R), B=TextureSample(UVs=@n27) {texture=&T_NoisePack_01}.G)
   B=Multiply(A=Power(Base=@n28.VGradient, Exp=@n29)) {constB=0.5}
  )
  n31 = MaterialFunctionCall(
   In=Power(Base=@n30, Exp=ScalarParameter() {defaultValue=0.9798, parameterName=Pow}) {constExponent=1}
   Contrast=ScalarParameter() {group=FlameShape, parameterName=HardEdgeness}
  ) {materialFunction=&CheapContrast}
  n32 = ScalarParameter() {defaultValue=0.6138, parameterName=Intensity_Distortion}
  n33 = TextureSampleParameter2D(UVs=Multiply(A=LinearInterpolate(A=@n12, B=@n13, Alpha=@n14), B=LinearInterpolate(A=@n31.Result, Alpha=@n32))) {
   parameterName=ReMappedColorTex
   samplerType=SAMPLERTYPE_LinearColor
   texture=/BRCosmetics/Weapons/FORT_Melee/Pickaxe_Raven_Quill_Parrot_Female/FX/T_Pickaxe_RavenQuilParrotF_lookup
  }
  n34 = TextureSample() {
   samplerType=SAMPLERTYPE_LinearGrayscale
   texture=/BRCosmetics/Characters/Player/Female/Medium/Bodies/F_MED_RavenQuill/Skin/Parrot_Wings/Texture/Ravenquil_Parrot_Gradient
  }
  n35 = LinearInterpolate(
   A=ScalarParameter() {defaultValue=-6.1367, parameterName=Gradient_Tweak_3}
   B=ScalarParameter() {defaultValue=-1.1186, parameterName=Gradient_Tweak_4}
   Alpha=@n34.R
  )
  n36 = Multiply(A=ComponentMask(None=@n10.B) {r=true}, B=ScalarParameter() {defaultValue=0.5337, parameterName=GradientInfluence})
  n37 = ScalarParameter() {defaultValue=3.4203, parameterName=Intensity_Distortion2}
  n38 = TextureSampleParameter2D(UVs=Multiply(A=Add(A=@n35, B=@n36), B=LinearInterpolate(A=@n19, Alpha=@n37))) {
   parameterName=ReMappedColorTex
   samplerType=SAMPLERTYPE_LinearColor
   texture=/BRCosmetics/Characters/Player/Female/Medium/Bodies/F_MED_RavenQuill/Skin/Parrot_Wings/Texture/Ravenquill_Gradient_2
  }
  n39 = LinearInterpolate(A=@n33.RGB, B=@n38.RGB, Alpha=@n10.R)
  n40 = MaterialFunctionCall(
   LDR Input=Clamp(None=@n39)
   HDR Tint=Multiply(A=@n39, B=VectorParameter() {group=FlameShape, parameterName=HDR_Tint}.RGB)
   HDR Intensity Multiplier=ScalarParameter() {defaultValue=0.5, group=FlameShape, parameterName=HDR_Intensity}
  ) {materialFunction=&DeriveHDRfromLDR}
  n41 = ShadingPathSwitch(
   Default=@n40.RemappedRange
   Deferred=@n40.RemappedRange
   Forward=@n40.RemappedRange
   Mobile=Multiply(A=ScalarParameter() {defaultValue=0.5, group=FlameShape, parameterName=MobileBoost}, B=@n39)
  )
  n42 = Multiply(A=ScalarParameter() {defaultValue=-0.3504, parameterName=Master_Emissive}, B=@n41)
  n43 = Multiply(
   A=Power(Base=@n9.BaseColor, Exp=ScalarParameter() {defaultValue=1, parameterName=ThinFilm_ParrotContrast})
   B=ScalarParameter() {defaultValue=1, parameterName=ThinFilm_ParrotAdd}
  )
  n44 = TextureCoordinate()
  n45 = ScalarParameter() {defaultValue=0.5, parameterName=Bias_Normal_Distortion}
  n46 = Subtract(A=@n18, B=AppendVector(A=@n45, B=@n45))
  n47 = ScalarParameter() {defaultValue=0.1, parameterName=Intensity_Normal_Distortion}
  n48 = TextureSampleParameter2D(UVs=Add(A=@n44, B=Multiply(A=@n46, B=@n47))) {
   parameterName=Normals
   samplerSource=SSM_Wrap_WorldGroupSettings
   samplerType=SAMPLERTYPE_Normal
   texture=/Game/Characters/Player/Female/Medium/Bodies/F_MED_RavenQuill/Skin/Parrot_Wings/Texture/T_F_MED_Raven_Quill_Parrot_Backpack_N
  }
  n49 = SetMaterialAttributes!(
   MaterialAttributes=@n9.MaterialAttributes
   EmissiveColor=Add(A=@n9.EmissiveColor, B=Multiply(A=@n11, B=@n42))
   BaseColor=LinearInterpolate(A=@n9.BaseColor, B=Add(A=@n43, B=@n41), Alpha=@n11)
   Normal=LinearInterpolate(A=@n9.Normal, B=@n48.RGB, Alpha=@n11)
   Metallic=LinearInterpolate(A=@n9.Metallic, B=ScalarParameter() {defaultValue=0.3593, parameterName=MetalicFlames}, Alpha=@n11)
   Roughness=LinearInterpolate(A=@n9.Roughness, B=ScalarParameter() {defaultValue=0.5313, parameterName=RoughnessFlames}, Alpha=@n11)
  ) {attributeSetTypes [
    #2
    #1
    #3
    #4
    #5
   ]}
  FunctionOutput(None=@n49)
  n50 = VectorParameter() {parameterName=Colorbones, defaultValue=<rgba 1 0 0.0316 1>}
  LinearInterpolate(A=Constant3Vector(), B=LinearInterpolate(A=@n9.BaseColor, B=Multiply(A=@n50.RGB, B=@n9.BaseColor), Alpha=@n10.R), Alpha=@n11)
  Multiply(A=@n14, B=@n39)
  Multiply(A=@n44, B=LinearInterpolate(A=@n30, Alpha=@n47))
  OneMinus(None=@n11)
  ScalarParameter() {defaultValue=-1.767, parameterName=UV_Distort_Rotator}
  StaticBool()
  TextureCoordinate() {coordinateIndex=1}
  TextureCoordinate() {uTiling=4}
  TextureCoordinate() {coordinateIndex=1, vTiling=0.5}
  TextureCoordinate() {uTiling=6, vTiling=1.5}
  TextureSample(UVs=@n6.Radial Coordinates) {texture=/Game/Packages/Fortress_SharedMaterials/Textures/T_UV_Checker_D}
 ]}
